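# Animation Blueprint for the M_LRG_EnsureHall_Snag player body cosmetic, which follows its attached parent mesh and adds Control Rig adjustments.
# The AnimGraph copies the parent's pose into the M_LRG_EnsureHall_Snag Control Rig, then outputs the result.
# Each update sets Dyn Crouch, Dyn Skydive and Dyn Emote to 1 or 0 from the crouching, skydiving and emote flags, feeding them to the rig as inputs.
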AnimBlueprint /BRCosmetics/Characters/Player/Male/Large/Bodies/M_LRG_EnsureHall_Snag/Meshes/M_LRG_EnsureHall_Snag_AnimBP
target_skeleton=&/M_LRG_EnsureHall_Snag_Skeleton
graphs [
 AnimGraph {nodes [
   n1 = Control Rig!(
    Source=Copy Pose From Mesh() {bUseAttachedParent=true}
    Alpha
    dyn_crouch:double=Get Dyn Crouch() -> (Dyn Crouch:double)
    dyn_skydive:double=Get Dyn Skydive() -> (Dyn Skydive:double)
    dyn_emote:double=Get Dyn Emote() -> (Dyn Emote:double)
   ) -> (Pose) {
    alphaBoolBlend {blendOption=Linear}
    controlRigAssetReference {blueprintRigClass=&/M_LRG_EnsureHall_Snag_CtrlRig.M_LRG_EnsureHall_Snag_CtrlRig_C}
   }
   Output Pose(Result=@n1) {layerGroup=None}
  ]}
 EventGraph {nodes [
   Event BlueprintUpdateAnimation()
   > n1 = Sequence!(execute) -> (then_0, then_1, then_2:Exec)
   > n2 = Branch(Condition=Get bIsCrouching())
   > Set Dyn Crouch(Dyn Crouch:double=1) -> (Output_Get:double)
   Set Dyn Crouch(execute=@n2.else, Dyn Crouch=0)
   n3 = Branch(execute=@n1.then_1, Condition=Get bIsSkyDiving())
   > Set Dyn Skydive(Dyn Skydive:double=1) -> (Output_Get:double)
   Set Dyn Skydive(execute=@n3.else, Dyn Skydive=0)
   n4 = Branch(execute=@n1.then_2, Condition=Get bIsPlayingEmote())
   > Set Dyn Emote(Dyn Emote:double=1) -> (Output_Get:double)
   Set Dyn Emote(execute=@n4.else, Dyn Emote=0)
  ]}
]
variables [
 Dyn Crouch {type=number, default=1}
 Dyn Emote {type=number, default=1}
 Dyn Skydive {type=number, default=1}
]
