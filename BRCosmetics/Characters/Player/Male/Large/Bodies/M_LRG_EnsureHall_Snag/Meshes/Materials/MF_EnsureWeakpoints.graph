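# Material function that layers glowing, corrupted weakpoints onto the EnsureHall Snag character body's material attributes.
# Eight custom primitive data vectors drive pre-skinned-space sphere masks, combined by Max through a vertex interpolator, with radii lerped over time from GameTimeFromBP.
# Mask shifting, panning triplanar noise, a hue-shifted heat gradient and flake normals feed reroutes for darkened base color, roughness, normal and edge emissive.
# CurrentStage remaps are computed but left unconnected.
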MaterialFunction /BRCosmetics/Characters/Player/Male/Large/Bodies/M_LRG_EnsureHall_Snag/Meshes/Materials/MF_EnsureWeakpoints
graph {nodes [
  Frac(None=Multiply(A=Time()))
  n1 = GetMaterialAttributes(None=FunctionInput() {inputType=FunctionInput_MaterialAttributes}) {attributeGetTypes [
    #1
    #2
    #3
    #4
    #5
    #6
   ]}
  n2 = NamedRerouteDeclaration(None=@n1.BaseColor) {name=IN_BC, nodeColor=<rgba 0.2656 0 1 1>}
  n3 = SmoothStep(Value=OneMinus(None=NamedRerouteUsage()))
  n4 = SetMaterialAttributes!(
   MaterialAttributes=@n1.MaterialAttributes
   BaseColor=LinearInterpolate(A=@n2, B=Multiply(A=@n2) {constB=0.2}, Alpha=@n3) {constB=0.2}
   Metallic
   Specular
   Roughness=LinearInterpolate(A=NamedRerouteUsage(), Alpha=@n3) {constB=0.8}
   EmissiveColor=NamedRerouteUsage()
   Normal=LinearInterpolate(A=NamedRerouteUsage(), B=NamedRerouteUsage(), Alpha=NamedRerouteUsage())
  ) {attributeSetTypes [
    #1
    #2
    #3
    #4
    #6
    #5
   ]}
  FunctionOutput(None=@n4)
  n5 = TextureObjectParameter() {group=Weakpoints, parameterName=Flake_Normal, samplerType=SAMPLERTYPE_Normal, texture=/Game/Global/Textures/Default/Noise/T_Noise_Normal_01}
  n6 = MaterialFunctionCall(
   TextureObject=@n5
   TextureSize=Constant() {r=4}
   WorldPosition=NamedRerouteUsage()
   World Space Normal=NamedRerouteUsage()
   ProjectionTransitionContrast=Constant() {r=1}
  ) {materialFunction=&WorldAlignedTexture}
  MaterialFunctionCall(Normal=@n6.XYZ Texture, Flatness=Constant() {r=-1}) {materialFunction=&FlattenNormal}
  n7 = NamedRerouteUsage()
  n8 = ScalarParameter() {defaultValue=1, group=Weakpoints, parameterName=HeatIntensity}
  n9 = Multiply(A=NamedRerouteUsage(), B=VectorParameter() {group=Weakpoints, parameterName=WeakpointEdgeColor, defaultValue=<rgba 1 0.0667 0.6791 1>}.RGB)
  NamedRerouteDeclaration(None=Add(A=Multiply(A=@n7, B=@n8) {constA=1, desc=emissive}, B=Multiply(A=@n9) {constB=40})) {name=Out_Emissive, nodeColor=<rgba 0.2422 0 1 1>}
  n10 = VectorParameter() {bUseCustomPrimitiveData=true, parameterName=Weakspot_01_Pos}
  n11 = Step(X=@n10.A) {constX=0, constY=0.1}
  n12 = LinearInterpolate(A=Subtract(A=@n11), B=@n11, Alpha=NamedRerouteUsage()) {desc=Native lerp toward}
  n13 = VectorParameter() {bUseCustomPrimitiveData=true, parameterName=Weakspot_02_Pos, primitiveDataIndex=4}
  n14 = Step(X=@n13.A) {constX=0, constY=0.1}
  n15 = LinearInterpolate(A=Subtract(A=@n14), B=@n14, Alpha=NamedRerouteUsage()) {desc=Native lerp toward}
  n16 = Max(
   A=SphereMask(
    A=LocalPosition() {localOrigin=InstancePreSkinning}.XYZ
    B=ComponentMask(None=@n10.RGB) {b=true, g=true, r=true}
    Radius=Multiply(A=@n10.A, B=@n12) {constA=20}
   ) {attenuationRadius=20, hardnessPercent=0}
   B=SphereMask(
    A=LocalPosition() {localOrigin=InstancePreSkinning}.XYZ
    B=ComponentMask(None=@n13.RGB) {b=true, g=true, r=true}
    Radius=Multiply(A=@n13.A, B=@n15) {constA=20}
   ) {attenuationRadius=20, hardnessPercent=0}
  )
  n17 = VectorParameter() {bUseCustomPrimitiveData=true, parameterName=Weakspot_03_Pos, primitiveDataIndex=8}
  n18 = Step(X=@n17.A) {constX=0, constY=0.1}
  n19 = Subtract(A=@n18)
  n20 = NamedRerouteUsage()
  n21 = SphereMask(
   A=LocalPosition() {localOrigin=InstancePreSkinning}.XYZ
   B=ComponentMask(None=@n17.RGB) {b=true, g=true, r=true}
   Radius=Multiply(A=@n17.A, B=LinearInterpolate(A=@n19, B=@n18, Alpha=@n20) {desc=Native lerp toward}) {constA=20}
  ) {attenuationRadius=20, hardnessPercent=0}
  n22 = LocalPosition() {localOrigin=InstancePreSkinning}
  n23 = VectorParameter() {bUseCustomPrimitiveData=true, parameterName=Weakspot_04_Pos, primitiveDataIndex=12}
  n24 = ComponentMask(None=@n23.RGB) {b=true, g=true, r=true}
  n25 = Step(X=@n23.A) {constX=0, constY=0.1}
  n26 = Multiply(A=@n23.A, B=LinearInterpolate(A=Subtract(A=@n25), B=@n25, Alpha=NamedRerouteUsage()) {desc=Native lerp toward}) {constA=20}
  n27 = VectorParameter() {bUseCustomPrimitiveData=true, parameterName=Weakspot_05_Pos, primitiveDataIndex=16}
  n28 = Step(X=@n27.A) {constX=0, constY=0.1}
  n29 = LinearInterpolate(A=Subtract(A=@n28), B=@n28, Alpha=NamedRerouteUsage()) {desc=Native lerp toward}
  n30 = Max(
   A=Max(A=Max(A=@n16, B=@n21), B=SphereMask(A=@n22.XYZ, B=@n24, Radius=@n26) {attenuationRadius=20, hardnessPercent=0})
   B=SphereMask(
    A=LocalPosition() {localOrigin=InstancePreSkinning}.XYZ
    B=ComponentMask(None=@n27.RGB) {b=true, g=true, r=true}
    Radius=Multiply(A=@n27.A, B=@n29) {constA=20}
   ) {attenuationRadius=20, hardnessPercent=0}
  )
  n31 = VectorParameter() {bUseCustomPrimitiveData=true, parameterName=Weakspot_06_Pos, primitiveDataIndex=20}
  n32 = Step(X=@n31.A) {constX=0, constY=0.1}
  n33 = Subtract(A=@n32)
  n34 = NamedRerouteUsage()
  n35 = SphereMask(
   A=LocalPosition() {localOrigin=InstancePreSkinning}.XYZ
   B=ComponentMask(None=@n31.RGB) {b=true, g=true, r=true}
   Radius=Multiply(A=@n31.A, B=LinearInterpolate(A=@n33, B=@n32, Alpha=@n34) {desc=Native lerp toward}) {constA=20}
  ) {attenuationRadius=20, hardnessPercent=0}
  n36 = LocalPosition() {localOrigin=InstancePreSkinning}
  n37 = VectorParameter() {bUseCustomPrimitiveData=true, parameterName=Weakspot_07_Pos, primitiveDataIndex=24}
  n38 = ComponentMask(None=@n37.RGB) {b=true, g=true, r=true}
  n39 = Step(X=@n37.A) {constX=0, constY=0.1}
  n40 = Multiply(A=@n37.A, B=LinearInterpolate(A=Subtract(A=@n39), B=@n39, Alpha=NamedRerouteUsage()) {desc=Native lerp toward}) {constA=20}
  n41 = VectorParameter() {bUseCustomPrimitiveData=true, parameterName=Weakspot_08_Pos, primitiveDataIndex=28}
  n42 = Step(X=@n41.A) {constX=0, constY=0.1}
  n43 = LinearInterpolate(A=Subtract(A=@n42), B=@n42, Alpha=NamedRerouteUsage()) {desc=Native lerp toward}
  n44 = Max(
   A=Max(A=Max(A=@n30, B=@n35), B=SphereMask(A=@n36.XYZ, B=@n38, Radius=@n40) {attenuationRadius=20, hardnessPercent=0})
   B=SphereMask(A=LocalPosition() {localOrigin=InstancePreSkinning}.XYZ, B=@n41.RGB, Radius=Multiply(A=@n41.A, B=@n43) {constA=20}) {attenuationRadius=20, hardnessPercent=0}
  )
  n45 = ShaderStageSwitch(PixelShader=VertexInterpolator(VS=@n44).PS, VertexShader=@n44)
  n46 = Divide(A=Time())
  n47 = Multiply(A=Subtract(A=Sine(None=@n46))) {constB=0.02}
  NamedRerouteDeclaration(None=Saturate(None=Add(A=@n45, B=@n47))) {name=M_Weakpoint, nodeColor=<rgba 1 0 0.3984 1>}
  NamedRerouteDeclaration(None=@n6.XYZ Texture) {name=FlakeNormal, nodeColor=<rgba 0.0547 0 1 1>}
  n48 = NamedRerouteUsage()
  n49 = NamedRerouteUsage()
  n50 = OneMinus(None=@n49)
  n51 = Constant() {r=0.5}
  n52 = Constant() {r=0.5}
  n53 = Constant() {r=1}
  NamedRerouteDeclaration(None=Saturate(None=MaterialFunctionCall!(Mask In=@n48, Amount=@n50, Range=@n51, Start Offset=@n52, Sharpness=@n53) {materialFunction=/Game/Global/MaterialFunctions/MF_MaskShifting}.Result)) {name=CorruptedArea, nodeColor=<rgba 0.2656 1 0 1>}
  NamedRerouteDeclaration(None=VectorParameter() {parameterName=Debug_Pos, defaultValue=<rgba 15.6945 -0.3882 127.1323 10>}.RGB) {name=Debugpos, nodeColor=<rgba 0 0.1094 1 1>}
  n54 = NamedRerouteUsage()
  n55 = SmoothStep(
   Min=ScalarParameter() {defaultValue=0.5, group=Weakpoints, parameterName=HardMask_Min, sortPriority=100}
   Max=ScalarParameter() {defaultValue=1, group=Weakpoints, parameterName=HardMask_Max, sortPriority=100}
   Value=Saturate(None=@n54)
  )
  NamedRerouteDeclaration(None=Saturate(None=StaticSwitchParameter(True=@n55, False=@n54) {defaultValue=true, group=Weakpoints, parameterName=GM_UseHardEdgedMask})) {name=BodyMask, nodeColor=<rgba 0 1 0.6016 1>}
  n56 = Constant() {r=0.5}
  n57 = SmoothStep(
   Min=ScalarParameter() {group=Weakpoints, parameterName=HardMask_Min_Emissive, sortPriority=101}
   Max=ScalarParameter() {defaultValue=0.2, group=Weakpoints, parameterName=HardMask_Max_Emissive, sortPriority=101}
   Value=NamedRerouteUsage()
  )
  n58 = OneMinus(None=@n57)
  NamedRerouteDeclaration(None=Multiply(A=If(A=@n57, B=@n56, "A > B"=@n58, "A == B"=@n58, "A < B"=@n57), B=NamedRerouteUsage())) {name=EdgeMask, nodeColor=<rgba 1 0.5625 0 1>}
  NamedRerouteDeclaration(None=@n1.Normal) {name=IN_Normal, nodeColor=<rgba 0 1 0.6719 1>}
  NamedRerouteDeclaration(None=@n1.Metallic) {name=IN_Metallic, nodeColor=<rgba 0.0547 1 0 1>}
  NamedRerouteDeclaration(None=@n1.Specular) {name=IN_Specular, nodeColor=<rgba 1 0 0.375 1>}
  NamedRerouteDeclaration(None=@n1.Roughness) {name=IN_Roughness, nodeColor=<rgba 0 0.6953 1 1>}
  n59 = Subtract(A=Time(), B=ScalarParameter() {defaultValue=-99999, group=Weakpoints, parameterName=GameTimeFromBP, sortPriority=100})
  n60 = Add(A=ScalarParameter() {defaultValue=1, group=Weakpoints, parameterName=Animation_Duration, sortPriority=101}) {constB=0}
  NamedRerouteDeclaration(None=Saturate(None=Divide(A=@n59, B=@n60))) {name=TimeNative, nodeColor=<rgba 0.2891 1 0 1>}
  NamedRerouteDeclaration(None=VertexInterpolator(VS=LocalPosition() {localOrigin=InstancePreSkinning}.XYZ).PS) {name=PreskinnedPos, nodeColor=<rgba 1 0 1 1>}
  n61 = NamedRerouteUsage()
  n62 = NamedRerouteUsage()
  NamedRerouteDeclaration(None=Saturate(None=Subtract(A=@n61, B=@n62))) {name=TransitionGradient, nodeColor=<rgba 1 0.1875 0 1>}
  NamedRerouteDeclaration(None=@n1.EmissiveColor) {name=IN_Emissive, nodeColor=<rgba 0 0.8125 1 1>}
  n63 = Add(A=NamedRerouteUsage()) {constB=0.1}
  n64 = Divide(A=Constant() {r=1}, B=Ceil(None=Constant() {r=1}))
  n65 = Multiply(A=@n64) {constB=0.5}
  n66 = Multiply(A=@n64, B=Ceil(None=Constant()))
  n67 = TextureSampleParameter2D(UVs=AppendVector(A=ComponentMask(None=@n63) {r=true}, B=Add(A=@n65, B=@n66))) {
   desc=Sampler Source: Clamp
   group=Weakpoints
   parameterName=Corruption_Gradient
   samplerSource=SSM_Clamp_WorldGroupSettings
   texture=/BRCosmetics/Accessories/FORT_Backpacks/Backpack_M_MED_VelvetDeskFiber/Textures/T_Heat_Gradient
  }
  n68 = MaterialFunctionCall(Hue Shift Percentage=Constant() {r=0.8}, Texture=@n67.RGB) {materialFunction=&HueShift}
  NamedRerouteDeclaration(None=Multiply(A=Multiply(A=@n68.Result) {constB=10}, B=NamedRerouteUsage())) {name=HeatEmissive, nodeColor=<rgba 1 0.0536 0 1>}
  NamedRerouteDeclaration(None=VertexInterpolator(VS=PreSkinnedNormal()).PS) {name=PreSkinnedNormal, nodeColor=<rgba 0.25 0.25 1 1>}
  n69 = MaterialFunctionCall(Mask In=NamedRerouteUsage(), Amount=OneMinus(None=@n49), Range=Constant() {r=0.1}, Start Offset=Constant() {r=1}, Sharpness=Constant() {r=1}) {materialFunction=&MF_MaskShifting}
  NamedRerouteDeclaration(None=OneMinus(None=Saturate(None=@n69.Result))) {name=CorruptedWeakpoint_Mask, nodeColor=<rgba 0 1 0.6953 1>}
  n70 = TextureObjectParameter() {group=Weakpoints, parameterName=CorruptNoise_Map, texture=/Game/Global/Textures/Default/Caustics/T_Turbulence_Seamless}
  n71 = VectorParameter() {group=Weakpoints, parameterName=CorruptNoise_Scale, defaultValue=<rgba 20 20 30 1>}
  n72 = NamedRerouteUsage()
  n73 = NamedRerouteUsage()
  n74 = Constant() {r=1}
  n75 = Constant2Vector() {g=-0.01}
  n76 = MaterialFunctionCall!(
   Export Float 4
   TextureObject=@n70
   TextureSize=@n71.RGB
   WorldPosition=@n72
   World Space Normal=@n73
   ProjectionTransitionContrast=@n74
   Pan_YZ=@n75
   Pan_XZ=@n75
   Pan_XY
  ) {materialFunction=/BRCosmetics/Gadgets/Assets/VinderTech_GliderChute/Glider_MissusMind/Materials/WorldAlignedTexture_ScalePanXYZ}
  n77 = Multiply(A=@n75) {constB=1.5}
  n78 = MaterialFunctionCall(
   TextureObject=@n70
   TextureSize=Multiply(A=@n71.RGB) {constB=2}
   WorldPosition=@n72
   World Space Normal=@n73
   ProjectionTransitionContrast=@n74
   Pan_YZ=@n77
   Pan_XZ=@n77
  ) {materialFunction=&WorldAlignedTexture_ScalePanXYZ}
  NamedRerouteDeclaration(None=Multiply(A=Add(A=@n76.XYZ Texture, B=@n78.XYZ Texture), B=NamedRerouteUsage())) {name=TriplannarNoise, nodeColor=<rgba 0 1 0.6484 1>}
  Power(Base=@n6.XYZ Texture)
  n79 = ScalarParameter() {group=Weakpoints, parameterName=CurrentStage, sortPriority=3}
  n80 = LinearInterpolate(A=Subtract(A=@n79), B=@n79, Alpha=NamedRerouteUsage()) {desc=Native lerp toward}
  n81 = Constant4Vector() {desc=Stage 4 to 5, constant=<rgba 4 5 0 1>}
  Saturate(None=MaterialFunctionCall(Input=@n80, Input Low=@n81.R, Input High=@n81.R, Target Low=@n81.R, Target High=@n81.R) {materialFunction=&RemapValueRange}.Result)
  n82 = Constant4Vector() {desc=Stage 0 to 1, constant=<rgba 0 1 0 1>}
  Saturate(None=MaterialFunctionCall(Input=@n80, Input Low=@n82.R, Input High=@n82.R, Target Low=@n82.R, Target High=@n82.R) {materialFunction=&RemapValueRange}.Result)
  n83 = Constant4Vector() {desc=Stage 1 to 2, constant=<rgba 1 2 0 1>}
  Saturate(None=MaterialFunctionCall(Input=@n80, Input Low=@n83.R, Input High=@n83.R, Target Low=@n83.R, Target High=@n83.R) {materialFunction=&RemapValueRange}.Result)
  n84 = Constant4Vector() {desc=Stage 2 to 3, constant=<rgba 2 3 0 1>}
  Saturate(None=MaterialFunctionCall(Input=@n80, Input Low=@n84.R, Input High=@n84.R, Target Low=@n84.R, Target High=@n84.R) {materialFunction=&RemapValueRange}.Result)
  n85 = Constant4Vector() {desc=Stage 3 to 4, constant=<rgba 4 5 0 1>}
  Saturate(None=MaterialFunctionCall(Input=@n80, Input Low=@n85.R, Input High=@n85.R, Target Low=@n85.R, Target High=@n85.R) {materialFunction=&RemapValueRange}.Result)
 ]}
