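# Control Rig for the face accessory part of the M_LRG_EnsureHall_Snag character, limiting head and neck rotation.
# It reads head, neck_02 and neck_01 local rotations as Euler angles, clamps Z, damps neck Y, and writes them back.
# Correction weight falls as dyn_emote rises, with skydive variables further reducing the head's.
# A final neck_01 rotation offset blends in with dyn_skydive_floating; three Multiply nodes are unused.
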ControlRigBlueprint /BRCosmetics/Characters/Player/Male/Large/Bodies/M_LRG_EnsureHall_Snag/Meshes/Parts/M_LRG_EnsureHall_Snag_FaceAcc_CtrlRig
graphs [
 RigVMFunctionLibrary {links [], nodes []}
 RigVMModel {nodes [
   RigUnit_BeginExecution = &RigUnit_BeginExecution()
   n1 = RigVMVariableNode() {variable=dyn_skydive}
   Add = &RigVMFunction_MathDoubleAdd(
    B=&RigVMFunction_MathDoubleSub(A=1, B=RigVMVariableNode() {variable=dyn_skydive_floating}.Value).Result
    A=&RigVMFunction_MathDoubleSub(A=1.5, B=@n1.Value).Result
   )
   n2 = &RigVMFunction_MathDoubleClamp(Maximum=&RigVMFunction_MathDoubleClamp(Value=@Add.Result).Result, Value=@n1.Value)
   Subtract = &RigVMFunction_MathDoubleSub(A=1, B=RigVMVariableNode() {variable=dyn_emote}.Value)
   RigUnit_GetTransform = &RigUnit_GetTransform(Item=(Type=Bone,Name="head"), Space=LocalSpace, bInitial=False)
   RigVMFunction_MathQuaternionToEuler = &RigVMFunction_MathQuaternionToEuler(RotationOrder=ZYX, Value=@RigUnit_GetTransform.Transform.Rotation)
   Clamp = &RigVMFunction_MathDoubleClamp(Maximum=2.5, Minimum=-5, Value=@RigVMFunction_MathQuaternionToEuler.Result.Z)
   Multiply = &RigVMFunction_MathDoubleMul(B=1, A=@RigVMFunction_MathQuaternionToEuler.Result.X)
   RigVMFunction_MathQuaternionFromEuler = &RigVMFunction_MathQuaternionFromEuler(
    Euler=(X=0,Y=0,Z=0)
    Euler.Z=@Clamp.Result
    Euler.X=@Multiply.Result
    Euler.Y=&RigVMFunction_MathDoubleMul(B=1, A=@RigVMFunction_MathQuaternionToEuler.Result.Y).Result
   )
   RigUnit_SetRotation = &RigUnit_SetRotation(
    Item=(Type=Bone,Name="head")
    bInitial=False
    bPropagateToChildren=true
    Weight=&RigVMFunction_MathDoubleClamp(Value=&RigVMFunction_MathDoubleSub(B=@n2.Result, A=@Subtract.Result).Result).Result
    ExecutePin=@RigUnit_BeginExecution.ExecutePin
    Value=@RigVMFunction_MathQuaternionFromEuler.Result
   )
   n3 = &RigVMFunction_MathQuaternionToEuler(
    RotationOrder=ZYX
    Value=&RigUnit_GetTransform(Item=(Type=Bone,Name="neck_02"), Space=LocalSpace, bInitial=False).Transform.Rotation
   )
   > &RigUnit_SetRotation(
    Item=(Type=Bone,Name="neck_02")
    bInitial=False
    Value=&RigVMFunction_MathQuaternionFromEuler(
     Euler=(X=0,Y=0,Z=0)
     Euler.Z=&RigVMFunction_MathDoubleClamp(Maximum=10, Minimum=-2.5, Value=@n3.Result.Z).Result
     Euler.X=&RigVMFunction_MathDoubleMul(B=1, A=@n3.Result.X).Result
     Euler.Y=&RigVMFunction_MathDoubleMul(B=0.75, A=@n3.Result.Y).Result
    ).Result
    Weight=@Subtract.Result
   )
   n4 = &RigVMFunction_MathQuaternionToEuler(
    RotationOrder=ZYX
    Value=&RigUnit_GetTransform(Item=(Type=Bone,Name="neck_01"), Space=LocalSpace, bInitial=False).Transform.Rotation
   )
   > &RigUnit_SetRotation(
    Item=(Type=Bone,Name="neck_01")
    bInitial=False
    Value=&RigVMFunction_MathQuaternionFromEuler(
     Euler=(X=0,Y=0,Z=0)
     Euler.Z=&RigVMFunction_MathDoubleClamp(Maximum=30, Value=@n4.Result.Z).Result
     Euler.X=&RigVMFunction_MathDoubleMul(B=1, A=@n4.Result.X).Result
     Euler.Y=&RigVMFunction_MathDoubleMul(B=0.15, A=@n4.Result.Y).Result
    ).Result
    Weight=@Subtract.Result
   )
   > &RigUnit_ModifyTransforms(
    ItemToModify=((Item=(Type=Bone,Name="neck_01"),Transform=(Rotation=(X=0,Y=0,Z=-0.2588,W=0.9659),Translation=(X=0,Y=0,Z=0),Scale3D=(X=1,Y=1,Z=1))))
    WeightMinimum=0
    Weight=RigVMVariableNode() {variable=dyn_skydive_floating}.Value
   )
   &RigVMFunction_MathDoubleMul(B=0.5, A=@RigVMFunction_MathQuaternionToEuler.Result.Z)
   &RigVMFunction_MathDoubleMul(B=0.5, A=@n3.Result.Z)
   &RigVMFunction_MathDoubleMul(B=0.5, A=@n4.Result.Z)
  ]}
]
hierarchy [
 root {loc=<0 0 0>, rot=<0 0 0 1>}
 pelvis {parent=root, loc=<0 -0.7321 92.4218>, rot=<0.0127 0.707 -0.0127 -0.707>}
 spine_01 {parent=pelvis, loc=<5.6002 -0.5128 0>, rot=<0 0 0.034 -0.9994>}
 spine_02 {parent=spine_01, loc=<8.0756 0.0139 0>, rot=<0 0 0.0238 -0.9997>}
 spine_03 {parent=spine_02, loc=<7.0395 0 0>, rot=<0 0 0.0149 -0.9999>}
 spine_04 {parent=spine_03, loc=<8.1023 0 0>, rot=<0 0 -0.0132 -0.9999>}
 spine_05 {parent=spine_04, loc=<10.0141 0 0>, rot=<0 0 -0.0015 -1>}
 neck_01 {parent=spine_05, loc=<10.815 1.9196 0>, rot=<0 0 -0.1864 -0.9825>}
 neck_02 {parent=neck_01, loc=<4.7347 0.3045 0>, rot=<0 0 0.0363 -0.9993>}
 head {parent=neck_02, loc=<5.7632 0.5552 0>, rot=<0 0 0.115 -0.9934>}
]
variables [
 dyn_emote {type=double, default=1}
 dyn_skydive {type=double, default=1}
 dyn_skydive_floating {type=double, default=0}
]
